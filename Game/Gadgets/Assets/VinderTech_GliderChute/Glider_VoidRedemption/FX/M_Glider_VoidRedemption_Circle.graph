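# Unlit, two-sided translucent material for Niagara mesh particles in the Void Redemption glider's circle effect.
# A layered smoothstep band across UV V is modulated by a borrowed character gradient texture, scrolled by dynamic-parameter speed and particle random, with strength set by UseCircleMask.
# The squared mask blends ShadowColor to particle color through an emissive-adjustment function.
# Particle alpha times the mask is faded by view angle where vertex red is painted.
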Material /Game/Gadgets/Assets/VinderTech_GliderChute/Glider_VoidRedemption/FX/M_Glider_VoidRedemption_Circle
graph {
 properties {
  bUsedWithNiagaraMeshParticles=true
  blendMode=BLEND_Translucent
  pixelDepthOffsetMode=PDOM_Legacy
  shadingModel=MSM_Unlit
  twoSided=true
  usedShadingModels=Unlit
 }
 nodes [
  Constant2Vector() {r=1}
  n1 = ParticleColor()
  n2 = ComponentMask(None=TextureCoordinate()) {g=true}
  n3 = OneMinus(None=Abs(None=Subtract(A=@n2) {constB=0.5}))
  n4 = Multiply(A=SmoothStep(Value=@n3) {constMin=0.5}) {constB=0.2}
  n5 = Multiply(A=SmoothStep(Value=@n3) {constMax=0.85, constMin=0.5}) {constB=0.25}
  n6 = Saturate(None=Add(A=SmoothStep(Value=@n3) {constMin=0.85}, B=Add(A=@n4, B=@n5)))
  n7 = TextureCoordinate()
  n8 = DynamicParameter() {paramNames [ScrollSpeed,Param2,Param3,Param4]}
  n9 = Constant()
  n10 = Multiply(A=Multiply(A=Time(), B=AppendVector(A=@n8.ScrollSpeed, B=@n9)), B=ParticleRandom())
  n11 = OneMinus(None=TextureSample(UVs=Add(A=@n7, B=@n10)) {texture=/Game/Characters/Player/Male/Medium/Bodies/M_MED_GelatoCourage/Skins/Elk/Textures/GelatoCourageElk_Gradient}.R)
  n12 = LinearInterpolate(A=@n6, B=Multiply(A=Power(Base=@n11) {constExponent=2.5}, B=@n6), Alpha=ScalarParameter() {defaultValue=1, parameterName=UseCircleMask})
  MaterialFunctionCall(Emissive=LinearInterpolate(A=VectorParameter() {parameterName=ShadowColor, defaultValue=<rgba 0.0261 0.0639 1 1>}.RGB, B=@n1.RGB, Alpha=Multiply(A=@n12, B=@n12))) {materialFunction=&MF_FX_EmissiveAdjustments}
  n13 = VertexColor()
  n14 = DotProduct(A=Multiply(A=VertexNormalWS(), B=@n13.R), B=CameraVectorWS())
  n15 = OneMinus(None=@n13.R)
  Multiply(A=@n1.A, B=Multiply(A=@n12, B=Add(A=@n14, B=@n15)))
  Panner() {speedX=1}
 ]
}
